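# MetaSound Source preset of MSS_NPCMusic_SunRose that supplies phase-based, attenuated stereo music for the Cerberus NPC.
# It overrides the parent's inputs with Cerberus loop waves, activation, transition and death stingers, mix levels, tempo and beat quantization.
# Parent inputs expose phase, activation, player-spotting and stop triggers alongside combat and occlusion state.
# The four phases reuse some loops and stingers, and several wave slots stay empty.
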MetaSoundSource /AIGods_Cerberus/Audio/Music/MSS_NPCMusic_SunRose_Cerberus_Preset
preset_parent=/SunRoseNPCMusic/Music/MSS_NPCMusic_SunRose
inputs [
 ActiveSound.bIsOccluded {type=Bool}
 ActiveSound.MaxDistance {type=Float}
 Debug {type=Bool}
 Mixing.DrumMult.Phase0 {type=Float}
 Mixing.DrumMult.Phase1 {type=Float}
 Mixing.DrumMult.Phase2 {type=Float}
 Mixing.DrumMult.Phase3 {type=Float}
 Mixing.Phase0 {type=Float}
 Mixing.Phase1 {type=Float}
 Mixing.Phase2 {type=Float}
 Mixing.Phase3 {type=Float}
 Mixing.Stinger.Activate {type=Float}
 Mixing.Stinger.Death {type=Float}
 Mixing.Stinger.Phase0 {type=Float}
 Mixing.Stinger.Phase1 {type=Float}
 Mixing.Stinger.Phase2 {type=Float}
 Mixing.Stinger.Phase3 {type=Float}
 NPC.Event.OnActivate {type=Trigger}
 NPC.Event.OnMusicStop {type=Trigger}
 NPC.Event.OnPhase0 {type=Trigger}
 NPC.Event.OnPhase1 {type=Trigger}
 NPC.Event.OnPhase2 {type=Trigger}
 NPC.Event.OnPhase3 {type=Trigger}
 NPC.Event.OnPlayerLost {type=Trigger}
 NPC.Event.OnPlayerSpotted {type=Trigger}
 NPC.Music.bEnableOcclusion {type=Bool}
 NPC.Music.Metronome {type=Float}
 NPC.Music.StartingPhase {type=Float}
 NPC.Music.VolumeRadius {type=Float}
 NPC.State.bInCombat {type=Bool}
 NPC.State.bListenForTargeting {type=Bool}
 Quantize.Phase0.LoopBeatDivision {type=Int32}
 Quantize.Phase0.StingerBeatDivision {type=Int32}
 Quantize.Phase1.LoopBeatDivision {type=Int32}
 Quantize.Phase1.StingerBeatDivision {type=Int32}
 Quantize.Phase2.LoopBeatDivision {type=Int32}
 Quantize.Phase2.StingerBeatDivision {type=Int32}
 Quantize.Phase3.LoopBeatDivision {type=Int32}
 Quantize.Phase3.StingerBeatDivision {type=Int32}
 Quantize.StingerDeath.StingerBeatDivision {type=Int32}
 UE.Attenuation.Distance {type=Float}
 UE.Source.OnPlay {type=Trigger}
 UE.Spatialization.Azimuth {type=Float}
 UE.Spatialization.Elevation {type=Float}
 Wave.Phase0.DrumLoop {type=WaveAsset}
 Wave.Phase0.ElseLoop {type=WaveAsset}
 Wave.Phase0.ExtLoop {type=WaveAsset}
 Wave.Phase1.DrumLoop {type=WaveAsset}
 Wave.Phase1.ElseLoop {type=WaveAsset}
 Wave.Phase1.ExtLoop {type=WaveAsset}
 Wave.Phase2.DrumLoop {type=WaveAsset}
 Wave.Phase2.ElseLoop {type=WaveAsset}
 Wave.Phase2.ExtLoop {type=WaveAsset}
 Wave.Phase3.DrumLoop {type=WaveAsset}
 Wave.Phase3.ElseLoop {type=WaveAsset}
 Wave.Phase3.ExtLoop {type=WaveAsset}
 Wave.Stinger.Activation {type=WaveAsset}
 Wave.Stinger.Death {type=WaveAsset}
 Wave.Stinger.Transition.Phase0 {type=WaveAsset}
 Wave.Stinger.Transition.Phase1 {type=WaveAsset}
 Wave.Stinger.Transition.Phase2 {type=WaveAsset}
 Wave.Stinger.Transition.Phase3 {type=WaveAsset}
]
interfaces [Fort.ActiveSound:1.0, UE.Attenuation:1.0, UE.OutputFormat.Stereo:1.0, UE.Source.OneShot:1.0, UE.Source:1.1, UE.Spatialization:1.0]
outputs [
 "UE.OutputFormat.Stereo.Audio:0" {type=Audio}
 "UE.OutputFormat.Stereo.Audio:1" {type=Audio}
 UE.Source.OneShot.OnFinished {type=Trigger}
]
pages [Default {input_defaults {
   ActiveSound.MaxDistance=0
   ActiveSound.bIsOccluded=null
   Debug=null
   Mixing.DrumMult.Phase0=1
   Mixing.DrumMult.Phase1=0.3
   Mixing.DrumMult.Phase2=0.5
   Mixing.DrumMult.Phase3=1
   Mixing.Phase0=1
   Mixing.Phase1=1.8
   Mixing.Phase2=1.5
   Mixing.Phase3=1.1
   Mixing.Stinger.Activate=2.5
   Mixing.Stinger.Death=3
   Mixing.Stinger.Phase0=1.2
   Mixing.Stinger.Phase1=1.8
   Mixing.Stinger.Phase2=1.7
   Mixing.Stinger.Phase3=2.2
   NPC.Event.OnActivate=null
   NPC.Event.OnMusicStop=null
   NPC.Event.OnPhase0=null
   NPC.Event.OnPhase1=null
   NPC.Event.OnPhase2=null
   NPC.Event.OnPhase3=null
   NPC.Event.OnPlayerLost=null
   NPC.Event.OnPlayerSpotted=null
   NPC.Music.Metronome=126
   NPC.Music.StartingPhase=0
   NPC.Music.VolumeRadius=10000
   NPC.Music.bEnableOcclusion="<MetaSoundBuilderResult.SUCCEEDED: 0>"
   NPC.State.bInCombat=null
   NPC.State.bListenForTargeting="<MetaSoundBuilderResult.SUCCEEDED: 0>"
   Quantize.Phase0.LoopBeatDivision=2
   Quantize.Phase0.StingerBeatDivision=2
   Quantize.Phase1.LoopBeatDivision=2
   Quantize.Phase1.StingerBeatDivision=2
   Quantize.Phase2.LoopBeatDivision=2
   Quantize.Phase2.StingerBeatDivision=2
   Quantize.Phase3.LoopBeatDivision=2
   Quantize.Phase3.StingerBeatDivision=2
   Quantize.StingerDeath.StingerBeatDivision=2
   UE.Attenuation.Distance=0
   UE.Source.OnPlay=null
   UE.Spatialization.Azimuth=0
   UE.Spatialization.Elevation=0
   Wave.Phase0.DrumLoop=null
   Wave.Phase0.ElseLoop=/AIGods_Cerberus/Audio/Music/Waves/MX_Cerberus_Phase_0_Loop_2mix
   Wave.Phase0.ExtLoop=null
   Wave.Phase1.DrumLoop=/AIGods_Cerberus/Audio/Music/Waves/MX_Cerberus_Phase_1_Loop_Drums
   Wave.Phase1.ElseLoop=&MX_Cerberus_Phase_0_Loop_2mix
   Wave.Phase1.ExtLoop=null
   Wave.Phase2.DrumLoop=/AIGods_Cerberus/Audio/Music/Waves/MX_Cerberus_Phase_02_Loop_Drums
   Wave.Phase2.ElseLoop=/AIGods_Cerberus/Audio/Music/Waves/MX_Cerberus_Phase_02_Loop_Else
   Wave.Phase2.ExtLoop=null
   Wave.Phase3.DrumLoop=&MX_Cerberus_Phase_1_Loop_Drums
   Wave.Phase3.ElseLoop=/AIGods_Cerberus/Audio/Music/Waves/MX_Cerberus_Phase_1_Loop_Else
   Wave.Phase3.ExtLoop=null
   Wave.Stinger.Activation=/AIGods_Cerberus/Audio/Music/Waves/MX_Cerberus_Effigy_Smash
   Wave.Stinger.Death=/AIGods_Cerberus/Audio/Music/Waves/MX_Cerberus_Death_Stinger
   Wave.Stinger.Transition.Phase0=null
   Wave.Stinger.Transition.Phase1=/AIGods_Cerberus/Audio/Music/Waves/MX_Cerberus_Phase_1_to_0_Stinger
   Wave.Stinger.Transition.Phase2=&MX_Cerberus_Phase_1_to_0_Stinger
   Wave.Stinger.Transition.Phase3=/AIGods_Cerberus/Audio/Music/Waves/MX_Cerberus_Phase_0_to_1_Stinger
  }}]
properties {
 assetImportData=&self:AssetImportData
 attenuationSettings=/SunRoseNPCMusic/Music/NPCMusic_SunRose_Att
 maxDistance=15400
 numChannels=2
 outputFormat=Stereo
 sampleRate=48000
 soundClassObject=&Fort_Music_NPC
 virtualizationMode=PlayWhenSilent
 volume=0.618
}
